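# Material function that builds an additional tiling surface layer for the Abducted set's chunk materials, defaulting to street road textures.
# Scaled UVs drive diffuse, mask and normal samples; a modulate input tints base color, and the mask's green channel feeds metallic, specular and roughness alike.
# A shading path switch gives Mobile untinted diffuse with constant surface values.
# A second output exposes diffuse alpha as a heightmap.
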MaterialFunction /Game/Environments/Apollo/Sets/Abducted/Chunks/Materials/MaterialFunctions/MF_Abducted_Chunk_AdditionalLayer
graph {nodes [
  n1 = Multiply(A=ScalarParameter() {defaultValue=13.1904, group=Additional Layer, parameterName=Additional Layer Scale}, B=TextureCoordinate())
  n2 = TextureSampleParameter2D(UVs=@n1) {
   group=Additional Layer
   parameterName=Additional Layer Diffuse
   samplerSource=SSM_Wrap_WorldGroupSettings
   texture=/Game/Packages/DS_Fortnight/SM/Textures/T_Street_Road_D
  }
  n3 = TextureSampleParameter2D(UVs=@n1) {
   group=Additional Layer
   parameterName=Additional Layer Specular
   samplerSource=SSM_Wrap_WorldGroupSettings
   samplerType=SAMPLERTYPE_Masks
   texture=/Game/Packages/DS_Fortnight/SM/Textures/T_Street_Road_M
  }
  n4 = TextureSampleParameter2D(UVs=@n1) {
   group=Additional Layer
   parameterName=Additional Layer Normal
   samplerSource=SSM_Wrap_WorldGroupSettings
   samplerType=SAMPLERTYPE_Normal
   texture=/Game/Packages/DS_Fortnight/SM/Textures/T_Street_Road_N
  }
  n5 = SetMaterialAttributes!(
   MaterialAttributes
   BaseColor=Multiply(A=FunctionInput() {inputName=Additional Layer Modulate}, B=@n2.RGB)
   Metallic=@n3.G
   Specular=@n3.G
   Roughness=@n3.G
   Normal=@n4.RGB
   EmissiveColor
  ) {attributeSetTypes [
    #1
    #2
    #3
    #4
    #5
    #6
   ]}
  n6 = SetMaterialAttributes(
   BaseColor=@n2.RGB
   Metallic=Constant()
   Specular=Constant() {r=0.5}
   Roughness=Constant() {r=1}
   Normal=Constant3Vector() {constant=<rgba 0.5 0.5 1 0>}
  ) {attributeSetTypes [
    #1
    #2
    #3
    #4
    #5
    #6
   ]}
  FunctionOutput(None=ShadingPathSwitch(Default=@n5, Deferred=@n5, Forward=@n5, Mobile=@n6))
  FunctionOutput(None=@n2.A) {outputName=heightmap}
 ]}
