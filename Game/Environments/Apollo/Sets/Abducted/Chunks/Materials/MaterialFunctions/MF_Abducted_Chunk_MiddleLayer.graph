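# Material function that builds the middle material layer for Abducted chunk environment pieces, using dirt textures by default.
# World-aligned projection at the Texture Scale input samples diffuse, packed spec and normal maps into material attributes, with tinted base color.
# Diffuse alpha feeds a separate Heightmap output.
# A world-aligned sediment tint, bypassed on mobile by a shading path switch, is built but left unconnected to either output.
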MaterialFunction /Game/Environments/Apollo/Sets/Abducted/Chunks/Materials/MaterialFunctions/MF_Abducted_Chunk_MiddleLayer
graph {nodes [
  n1 = VectorParameter() {group=Middle Layer, parameterName=Middle Layer Tint, defaultValue=<rgba 1 1 1 0>}
  n2 = TextureObjectParameter() {group=Middle Layer, parameterName=Middle Layer Diffuse, texture=/Game/Environments/Apollo/Sets/Terrain/Textures/T_Apollo_Dirt_D_02}
  n3 = TextureObjectParameter() {
   group=Middle Layer
   parameterName=Middle Layer Specular
   samplerType=SAMPLERTYPE_Masks
   texture=/Game/Environments/Apollo/Sets/Terrain/Textures/T_Apollo_Dirt_S_02
  }
  n4 = FunctionInput(Preview=Constant() {r=1000}) {inputName=Texture Scale}
  n5 = TextureObjectParameter() {
   group=Middle Layer
   parameterName="Middle Layer Normal "
   samplerType=SAMPLERTYPE_Normal
   texture=/Game/Environments/Apollo/Sets/Terrain/Textures/T_Apollo_Dirt_N_02
  }
  n6 = MaterialFunctionCall!(
   Diffuse Texture Object=@n2
   Spec Texture Object=@n3
   TextureSize=@n4
   --------------------------
   Use seperate settings for Z projection
   Use seperate texture for Z projection
   Texture Size Z projection=@n4
   Z Plane Diffuse Texture Object
   Z Plane Spec Texture Object
   "--------------------------      "
   ProjectionTransitionContrast
   Use Texture Alpha Heightmap to Refine the Transistion
   HeightLerpTransitionContrast
   PickTheDiffuseHeightChannel
   TheDiffuseHeightChannel
   "--------------------------       "
   NormalTextureObject=@n5
   World Space Normals=StaticBool()
   "--------------------------  "
   Export Float 4 Diffuse=StaticBool() {value=true}
   Export Float 4 Spec
   "--------------------------     "
   WorldPosition
   World Space Normal
  ) {materialFunction=/Engine/Functions/Engine_MaterialFunctions02/Texturing/WorldAlignedTextures_Complex}
  n7 = SetMaterialAttributes!(
   MaterialAttributes
   BaseColor=Multiply(A=ComponentMask(None=@n1.RGB) {b=true, g=true, r=true}, B=ComponentMask(None=@n6.XYZ Diffuse Texture) {b=true, g=true, r=true})
   Metallic=ComponentMask(None=@n6.XYZ Spec) {g=true}
   Specular=ComponentMask(None=@n6.XYZ Spec) {r=true}
   Roughness=ComponentMask(None=@n6.XYZ Spec) {b=true}
   Normal=@n6.Normal XYZ Texture
   Opacity
  ) {attributeSetTypes [
    #1
    #2
    #3
    #4
    #5
    #6
   ]}
  FunctionOutput(None=@n7)
  FunctionOutput(None=ComponentMask(None=@n6.XYZ Diffuse Texture) {a=true}) {outputName=Heightmap}
  n8 = Constant() {r=0.5}
  n9 = TextureObjectParameter() {group=Sediment, parameterName=Sediment Texture, texture=/Game/Environments/Apollo/Sets/Abducted/Chunks/Textures/T_Abducted_Sediment_B}
  n10 = MaterialFunctionCall(
   Export Float 4=StaticBool() {value=true}
   TextureObject=@n9
   TextureSize=ScalarParameter() {defaultValue=2000, group=Sediment, parameterName=Sediment_Tiling}
  ) {materialFunction=/Game/Weapons/WeaponSkins/Wraps/MaterialFunctions/WorldAlignedTexture}
  n11 = MaterialFunctionCall!(
   Material=@n7
   Tint=ComponentMask(None=Add(A=@n8, B=@n10.XY Texture)) {b=true, g=true, r=true}
   TintMask=ScalarParameter() {defaultValue=0.25, group=Sediment, parameterName=Sediment Amount}
  ) {materialFunction=/Engine/Functions/MaterialLayerFunctions/MatLayerBlend_Tint}
  ShadingPathSwitch(Default=@n11.Blended Material, Deferred=@n11.Blended Material, Forward=@n11.Blended Material, Mobile=@n7)
 ]}
