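# Material function that builds the top-layer material attributes for Abducted set chunks, defaulting to Apollo grass textures.
# Parameterized diffuse, specular and normal textures are world-aligned, with diffuse alpha output as a heightmap.
# World-aligned blurred noise blends two tints over the diffuse, while mobile uses only Color B.
# The specular texture's green channel drives Metallic, Specular and Roughness alike.
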MaterialFunction /Game/Environments/Apollo/Sets/Abducted/Chunks/Materials/MaterialFunctions/MF_Abducted_Chunk_Toplayer
graph {nodes [
  n1 = VectorParameter() {group=Top Layer, parameterName=Top Layer Modulate Color B, defaultValue=<rgba 1 1 1 1>}
  n2 = VectorParameter() {group=Top Layer, parameterName=Top Layer Modulate Color A, defaultValue=<rgba 1 1 1 1>}
  n3 = MaterialFunctionCall(
   TextureObject=TextureObject() {samplerType=SAMPLERTYPE_Grayscale, texture=&LowResBlurredNoise}
   TextureSize=ScalarParameter() {defaultValue=3000, group=Top Layer, parameterName=Modulate Mask Tiling}
  ) {materialFunction=&WorldAlignedTexture}
  n4 = TextureObjectParameter() {
   group=Top Layer
   parameterName=Top Layer Specular
   samplerType=SAMPLERTYPE_Masks
   texture=/Game/Environments/Apollo/Sets/Terrain/Textures/T_Apollo_Grass_D_02
  }
  n5 = TextureObjectParameter() {
   group=Top Layer
   parameterName=Top Layer Normal
   samplerType=SAMPLERTYPE_Normal
   texture=/Game/Environments/Apollo/Sets/Terrain/Textures/T_Apollo_Grass_N_01
  }
  n6 = MaterialFunctionCall!(
   Diffuse Texture Object=TextureObjectParameter() {group=Top Layer, parameterName=Top Layer Diffuse, texture=/Game/Environments/Apollo/Sets/Terrain/Textures/T_Apollo_Grass2_D}
   Spec Texture Object=@n4
   TextureSize=ScalarParameter() {defaultValue=1200, group=Top Layer, parameterName=Top Layer Scale}
   --------------------------
   Use seperate settings for Z projection
   Use seperate texture for Z projection
   Texture Size Z projection
   Z Plane Diffuse Texture Object
   Z Plane Spec Texture Object
   "--------------------------      "
   ProjectionTransitionContrast
   Use Texture Alpha Heightmap to Refine the Transistion
   HeightLerpTransitionContrast
   PickTheDiffuseHeightChannel
   TheDiffuseHeightChannel
   "--------------------------       "
   NormalTextureObject=@n5
   World Space Normals=StaticBool()
   "--------------------------  "
   Export Float 4 Diffuse=StaticBool() {value=true}
   Export Float 4 Spec
   "--------------------------     "
   WorldPosition
   World Space Normal
  ) {materialFunction=/Engine/Functions/Engine_MaterialFunctions02/Texturing/WorldAlignedTextures_Complex}
  n7 = ComponentMask(None=@n6.XYZ Diffuse Texture) {b=true, g=true, r=true}
  n8 = Multiply(A=ComponentMask(None=LinearInterpolate(A=@n1.RGB, B=@n2.RGB, Alpha=@n3.XYZ Texture)) {b=true, g=true, r=true}, B=@n7)
  n9 = MaterialFunctionCall(Float3=@n6.XYZ Spec) {materialFunction=&BreakOutFloat3Components}
  n10 = SetMaterialAttributes!(
   MaterialAttributes
   BaseColor=ShadingPathSwitch(Default=@n8, Deferred=@n8, Forward=@n8, Mobile=Multiply(A=@n1.RGB, B=@n7))
   Metallic=@n9.G
   Specular=@n9.G
   Roughness=@n9.G
   Normal=@n6.Normal XYZ Texture
  ) {attributeSetTypes [
    #1
    #2
    #3
    #4
    #5
   ]}
  FunctionOutput(None=@n10)
  FunctionOutput(None=ComponentMask(None=@n6.XYZ Diffuse Texture) {a=true}) {outputName=Top Layer Heightmap}
 ]}
