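# Layered Default Lit material for the Apollo Abducted chunk environment set, with shading path switches that cut cost on mobile.
# Baked diffuse, specular and normal maps blend with a middle layer via vertex color and noise breakup, plus detail normals off mobile.
# A world-aligned top layer blends over this using heightmaps.
# An optional additional layer applies only without baked textures, and is skipped on mobile unless forced.
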Material /Game/Environments/Apollo/Sets/Abducted/Chunks/Materials/MaterialFunctions/M_Abducted_Chunk_Mobile
graph {
 properties {bUseMaterialAttributes=true, pixelDepthOffsetMode=PDOM_Legacy, usedShadingModels=Default Lit}
 nodes [
  n1 = TextureSample(UVs=Multiply(A=ScalarParameter() {defaultValue=1, group=Break Up, parameterName=Breakup Texture Tiling}, B=TextureCoordinate())) {texture=/Game/Effects/Fort_Effects/Textures/Fire/T_FireNoise_02}
  n2 = ShadingPathSwitch(Default=@n1.R, Deferred=@n1.R, Forward=@n1.R)
  n3 = LinearInterpolate(A=ScalarParameter() {defaultValue=0.8032, group=Break Up, parameterName=Break Up Bias}, B=Constant() {r=1}, Alpha=@n2)
  n4 = VertexColor()
  n5 = Multiply(A=@n3, B=@n4.R)
  n6 = TextureSampleParameter2D() {
   group=Baked
   parameterName=Baked Diffuse
   texture=/Game/Environments/Apollo/Sets/Abducted/Chunks/Textures/Mothership/T_Abducted_Chunk_2_Base_D
  }
  n7 = TextureSampleParameter2D() {
   group=Baked
   parameterName=Baked Specular
   samplerType=SAMPLERTYPE_Masks
   texture=/Game/Environments/Apollo/Sets/Abducted/Chunks/Textures/Mothership/T_Abducted_Chunk_2_Base_S
  }
  n8 = TextureSampleParameter2D() {
   group=Baked
   parameterName=Baked Normal
   samplerType=SAMPLERTYPE_Normal
   texture=/Game/Environments/Apollo/Sets/Abducted/Chunks/Textures/Mothership/T_Abducted_Chunk_2_Base_N
  }
  n9 = SetMaterialAttributes!(
   MaterialAttributes
   Base Color=Multiply(A=VectorParameter() {group=Baked, parameterName=Baked Modulate color, defaultValue=<rgba 0.6354 0.5082 0.4071 1>}.RGB, B=@n6.RGB)
   Specular=@n7.R
   Roughness=@n7.R
   Normal=@n8.RGB
  ) {attributeSetTypes [
    #1
    #2
    #3
    #4
   ]}
  n10 = MaterialFunctionCall!(Texture Scale=ScalarParameter() {defaultValue=1021.8221, group=Middle Layer, parameterName=Middle Layer Texture Scale}) {materialFunction=&/MF_Abducted_Chunk_MiddleLayer}
  n11 = Multiply(A=@n3, B=@n4.B)
  n12 = MaterialFunctionCall!(Base Material=@n9, Top Material=@n10.Result, Alpha=@n11) {materialFunction=/Engine/Functions/MaterialLayerFunctions/MatLayerBlend_Standard}
  n13 = TextureSampleParameter2D(UVs=Multiply(A=ScalarParameter() {defaultValue=2, group=Detail Normal, parameterName=Detail Normal Tiling}, B=TextureCoordinate())) {
   group=Detail Normal
   parameterName=Detail Normal
   samplerType=SAMPLERTYPE_Normal
   texture=/Game/Environments/Apollo/Rocks/Natural_Rocks/Textures/T_RockDetails_02_N
  }
  n14 = MaterialFunctionCall!(
   NormalMask=Multiply(A=ScalarParameter() {defaultValue=1, group=Detail Normal, parameterName=Detail Normal Intensity}, B=@n5)
   Material=@n12.Blended Material
   Normal=@n13.RGB
  ) {materialFunction=/Engine/Functions/MaterialLayerFunctions/MatLayerBlend_NormalBlend}
  n15 = StaticSwitchParameter(
   True=StaticSwitchParameter(
    True=ShadingPathSwitch(Default=@n14.Blended Material, Deferred=@n14.Blended Material, Forward=@n14.Blended Material, Mobile=@n12.Blended Material)
    False=@n12.Blended Material
   ) {defaultValue=true, parameterName=Use Detail Normals?}
   False=@n10.Result
  ) {defaultValue=true, parameterName=Use Baked Textures}
  n16 = MaterialFunctionCall() {materialFunction=&/MF_Abducted_Chunk_Toplayer}
  n17 = Add(
   A=Multiply(A=ScalarParameter() {defaultValue=1, group=Top Layer, parameterName=Top Layer Blend Bias}, B=@n16.Top Layer Heightmap)
   B=Multiply(A=@n10.Heightmap, B=Constant() {r=-7})
  )
  n18 = MaterialFunctionCall!(
   In Explicit Normal=Constant3Vector() {constant=<rgba 0 0 1 0>}
   In World Vector
   Blend Sharpness=ScalarParameter() {defaultValue=7.6, group=Top Layer, parameterName=Top Layer Blend Sharpness}
   Blend Bias=Multiply(A=Add(A=@n5, B=@n11), B=Add(A=@n4.G, B=@n17))
   Clamped?=StaticBool() {value=true}
   Alpha
  ) {materialFunction=/Engine/Functions/Engine_MaterialFunctions01/AlphaBlend/WorldAlignedBlend}
  n19 = StaticSwitchParameter(
   True=MaterialFunctionCall(Base Material=@n15, Top Material=@n16.Result, Alpha=Clamp(None=@n18.w/Vertex Normals)) {materialFunction=&MatLayerBlend_Standard}.Blended Material
   False=@n15
  ) {defaultValue=true, parameterName=Use Top Layer?}
  n20 = VectorParameter() {group=Additional Layer, parameterName=Additional Layer Modulate, defaultValue=<rgba 1 1 1 1>}
  n21 = StaticSwitchParameter(True=Multiply(A=@n2, B=@n20.RGB), False=@n20.RGB) {defaultValue=true, group=Additional Layer, parameterName=Large Scale Break up}
  n22 = MaterialFunctionCall!(Additional Layer Modulate=ShadingPathSwitch(Default=@n21, Deferred=@n21, Forward=@n21, Mobile=@n20.RGB)) {materialFunction=&/MF_Abducted_Chunk_AdditionalLayer}
  n23 = Multiply(
   A=LinearInterpolate(
    A=ScalarParameter() {defaultValue=2, group=Additional Layer, parameterName=Additional Heightmap Bias B}
    B=ScalarParameter() {defaultValue=0.5, group=Additional Layer, parameterName=Additional Heightmap Bias A}
    Alpha=@n22.heightmap
   )
   B=Multiply(A=@n3, B=OneMinus(None=@n4.A))
  )
  n24 = ScalarParameter() {defaultValue=1, group=Additional Layer, parameterName=Additional Heightmap Contrast}
  n25 = StaticSwitchParameter(
   True=MaterialFunctionCall(
    Base Material=@n19
    Top Material=@n22.Result
    Alpha=MaterialFunctionCall(In=@n23, Contrast=@n24) {materialFunction=&CheapContrast}.Result
   ) {materialFunction=&MatLayerBlend_Standard}.Blended Material
   False=@n19
  ) {parameterName=Use Additional Layer?}
  StaticSwitchParameter(
   True=@n19
   False=ShadingPathSwitch(
    Default=@n25
    Deferred=@n25
    Forward=@n25
    Mobile=StaticSwitchParameter(True=@n25, False=@n19) {parameterName="Force Additional layer on mobile (expensive!)"}
   )
  ) {defaultValue=true, parameterName=Use Baked Textures}
 ]
}
